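; Fortnite physics tree analytics reporter that records when, where and by whom logs are felled or destroyed, plus damage they inflict.
; Report functions find or add per-instance records by name, accumulating damage for players, structures, vehicles and AI.
; After playlist data is ready, the server listens for game phase changes through the event router.
; At EndGame it hand-builds a bracketed string of keyless, quoted rows and fires Core.PhysicsTreeStats with session and playlist.
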(event EventEndPlay
  (params (EndPlayReason :type EEndPlayReason%20Enum :default "Destroyed"))
  (bind _gamephaseupdatedasynctask (Variables|Default|GetGamePhaseUpdatedAsyncTask)
    (:then
      (GameplayEventRouter|StopListeningtoStatefulEvent _gamephaseupdatedasynctask))
    (:else)))

(event EventBeginPlay
  (if (Networking|IsServer)
    (Class|FortAsyncActionCurrentPlaylistReady|PlaylistDataReadyAsync
      (:OnReady
        (bind _returnvalue (GameplayEventRouter|GetEventRouter "/Script/FortniteGame.FortGameState" self))
        (bind (_asynctaskproxy _payload) (GameplayEventRouter|StartListeningToStatefulEvent _returnvalue "/Script/FortniteGame.FortGamePhaseUpdatedEvent")
          (:then
            (Variables|Default|SetGamePhaseUpdatedAsyncTask _asynctaskproxy))
          (:OnEventReceived
            _)
          (:OnSavedEventStateExists
            _))
        (CallFunction|PhaseChanged (Utilities|Struct|BreakFortGamePhaseUpdatedEvent _payload))))))

(event Custom|EndMatchReport
  (CallFunction|ReportTreeAnalytics))

(event Custom|PhaseChanged
  (params (GamePhase :type EAthenaGamePhase%20Enum))
  (switch Utilities|FlowControl|Switch|SwitchonEAthenaGamePhase GamePhase
    (:EndGame
      (CallFunction|EndMatchReport))))

(fn GetLogDataByInstanceName
  (params (InstanceName :type String))
  (for (_array_element _array_index) (Variables|Default|GetPhysicsLogData)
    (if (Utilities|String|EqualExactly(String) InstanceName (Utilities|Struct|BreakPhysicsLogAnalyticsStruct _array_element))
      (return _array_element _array_index true)))
  (return 0 -1 false))

(fn ReportDamageInflicted
  (params (InstanceName :type String) (DamageCategory :type PhysicsLogAnalytics_DamageCategory%20Enum :default "NewEnumerator0") (Amount :type Float%20%28double-precision%29))
  (bind (_instancename_29_e13a3cf84b9151e7004d9b8c51fd00b9 _felledby_2_076924f44337fde90f8b5c9c7645cdf2 _felledatlocation_7_87d469db42bc3085780226ac3b154ad1 _felledatmatchtime_17_701a35db42f8da247cdc50ac58cccd4e _destroyedatlocation_33_93747681485fceb2f9d50b8f2c9ba72b _destroyedatmatchtime_31_1dcf4002426a45c86050d4b13c85632f) (Utilities|Struct|BreakPhysicsLogAnalyticsStruct (CallFunction|GetLogDatabyInstanceName InstanceName)))
  (bind (_instancename_29_e13a3cf84b9151e7004d9b8c51fd00b9_2 _felledby_2_076924f44337fde90f8b5c9c7645cdf2_2 _felledatlocation_7_87d469db42bc3085780226ac3b154ad1_2 _felledatmatchtime_17_701a35db42f8da247cdc50ac58cccd4e_2 _destroyedatlocation_33_93747681485fceb2f9d50b8f2c9ba72b_2 _destroyedatmatchtime_31_1dcf4002426a45c86050d4b13c85632f_2 _damagetoplayers_20_f701392440aac7ac33a8b2bcaeb3c5d0 _damagetostructures_22_636fabde409e1a51baf808863b05d213 _damagetovehicles_24_69a07f714b43df853ca180b441a6a29c _damagetoai_26_9671d0a943425bb2ec73159c0b19c99d) (Utilities|Struct|BreakPhysicsLogAnalyticsStruct (CallFunction|GetLogDatabyInstanceName InstanceName)))
  (bind (_data _index _success) (CallFunction|GetLogDatabyInstanceName InstanceName))
  (Utilities|Array|SetArrayElem (LocalVariables|GetDamageCounts) (Math|Conversions|ToInteger(Byte) DamageCategory) Amount)
  (if _success
    ;[1 1 1] Add new damage
    (for (_array_element _array_index) (LocalVariables|GetDamageCounts)
      (Utilities|Array|SetArrayElem (LocalVariables|GetDamageCounts) _array_index (+ (Utilities|Select :Index _array_index :Option%200 _damagetoplayers_20_f701392440aac7ac33a8b2bcaeb3c5d0 :Option%201 _damagetostructures_22_636fabde409e1a51baf808863b05d213 :Option%202 _damagetovehicles_24_69a07f714b43df853ca180b441a6a29c :Option%203 _damagetoai_26_9671d0a943425bb2ec73159c0b19c99d) _array_element)))
    (Utilities|Array|SetArrayElem (Variables|Default|GetPhysicsLogData) _index (Utilities|Struct|MakePhysicsLogAnalyticsStruct _instancename_29_e13a3cf84b9151e7004d9b8c51fd00b9 _felledby_2_076924f44337fde90f8b5c9c7645cdf2 _felledatlocation_7_87d469db42bc3085780226ac3b154ad1 _felledatmatchtime_17_701a35db42f8da247cdc50ac58cccd4e _destroyedatlocation_33_93747681485fceb2f9d50b8f2c9ba72b _destroyedatmatchtime_31_1dcf4002426a45c86050d4b13c85632f (Utilities|Array|Get(acopy) (LocalVariables|GetDamageCounts)) (Utilities|Array|Get(acopy) (LocalVariables|GetDamageCounts) 1) (Utilities|Array|Get(acopy) (LocalVariables|GetDamageCounts) 2) (Utilities|Array|Get(acopy) (LocalVariables|GetDamageCounts) 3)))
    (else
      (Utilities|Array|Add (Variables|Default|GetPhysicsLogData) (Utilities|Struct|MakePhysicsLogAnalyticsStruct (Variables|GetInstanceName) "" "0.000000,0.000000,0.000000" 0.0 "0.000000,0.000000,0.000000" 0.0 (Utilities|Array|Get(acopy) (LocalVariables|GetDamageCounts)) (Utilities|Array|Get(acopy) (LocalVariables|GetDamageCounts) 1) (Utilities|Array|Get(acopy) (LocalVariables|GetDamageCounts) 2) (Utilities|Array|Get(acopy) (LocalVariables|GetDamageCounts) 3))))))

(fn ReportLogDestroyed
  (params (Instance%20Name :type String) (Location :type Vector))
  (bind (_instancename_29_e13a3cf84b9151e7004d9b8c51fd00b9 _felledby_2_076924f44337fde90f8b5c9c7645cdf2 _felledatlocation_7_87d469db42bc3085780226ac3b154ad1 _felledatmatchtime_17_701a35db42f8da247cdc50ac58cccd4e _destroyedatlocation_33_93747681485fceb2f9d50b8f2c9ba72b _destroyedatmatchtime_31_1dcf4002426a45c86050d4b13c85632f _damagetoplayers_20_f701392440aac7ac33a8b2bcaeb3c5d0 _damagetostructures_22_636fabde409e1a51baf808863b05d213 _damagetovehicles_24_69a07f714b43df853ca180b441a6a29c _damagetoai_26_9671d0a943425bb2ec73159c0b19c99d) (Utilities|Struct|BreakPhysicsLogAnalyticsStruct (CallFunction|GetLogDatabyInstanceName Instance%20Name)))
  (bind (_data _index _success) (CallFunction|GetLogDatabyInstanceName Instance%20Name))
  (if _success
    (Utilities|Array|SetArrayElem (Variables|Default|GetPhysicsLogData) _index (Utilities|Struct|MakePhysicsLogAnalyticsStruct _instancename_29_e13a3cf84b9151e7004d9b8c51fd00b9 _felledby_2_076924f44337fde90f8b5c9c7645cdf2 _felledatlocation_7_87d469db42bc3085780226ac3b154ad1 _felledatmatchtime_17_701a35db42f8da247cdc50ac58cccd4e (Variables|GetLocation) (GameState|GetMatchTime (Fortnite|Athena|GameState|GetGameStateAthena)) _damagetoplayers_20_f701392440aac7ac33a8b2bcaeb3c5d0 _damagetostructures_22_636fabde409e1a51baf808863b05d213 _damagetovehicles_24_69a07f714b43df853ca180b441a6a29c _damagetoai_26_9671d0a943425bb2ec73159c0b19c99d))
    (else
      (Utilities|Array|Add (Variables|Default|GetPhysicsLogData) (Utilities|Struct|MakePhysicsLogAnalyticsStruct (Variables|GetInstance%20Name) "" "0.000000,0.000000,0.000000" 0.0 (Variables|GetLocation) (GameState|GetMatchTime (Fortnite|Athena|GameState|GetGameStateAthena)))))))

(fn ReportLogFelled
  (params (Instance%20Name :type String) (Felled%20By :type String) (Location :type Vector))
  (bind (_instancename_29_e13a3cf84b9151e7004d9b8c51fd00b9 _felledby_2_076924f44337fde90f8b5c9c7645cdf2 _felledatlocation_7_87d469db42bc3085780226ac3b154ad1 _felledatmatchtime_17_701a35db42f8da247cdc50ac58cccd4e _destroyedatlocation_33_93747681485fceb2f9d50b8f2c9ba72b _destroyedatmatchtime_31_1dcf4002426a45c86050d4b13c85632f _damagetoplayers_20_f701392440aac7ac33a8b2bcaeb3c5d0 _damagetostructures_22_636fabde409e1a51baf808863b05d213 _damagetovehicles_24_69a07f714b43df853ca180b441a6a29c _damagetoai_26_9671d0a943425bb2ec73159c0b19c99d) (Utilities|Struct|BreakPhysicsLogAnalyticsStruct (CallFunction|GetLogDatabyInstanceName Instance%20Name)))
  (bind (_data _index _success) (CallFunction|GetLogDatabyInstanceName Instance%20Name))
  (if _success
    (Utilities|Array|SetArrayElem (Variables|Default|GetPhysicsLogData) _index (Utilities|Struct|MakePhysicsLogAnalyticsStruct _instancename_29_e13a3cf84b9151e7004d9b8c51fd00b9 (Variables|GetFelled%20By) (Variables|GetLocation) (GameState|GetMatchTime (Fortnite|Athena|GameState|GetGameStateAthena)) _destroyedatlocation_33_93747681485fceb2f9d50b8f2c9ba72b _destroyedatmatchtime_31_1dcf4002426a45c86050d4b13c85632f _damagetoplayers_20_f701392440aac7ac33a8b2bcaeb3c5d0 _damagetostructures_22_636fabde409e1a51baf808863b05d213 _damagetovehicles_24_69a07f714b43df853ca180b441a6a29c _damagetoai_26_9671d0a943425bb2ec73159c0b19c99d))
    (else
      (Utilities|Array|Add (Variables|Default|GetPhysicsLogData) (Utilities|Struct|MakePhysicsLogAnalyticsStruct (Variables|GetInstance%20Name) (Variables|GetFelled%20By) (Variables|GetLocation) (GameState|GetMatchTime (Fortnite|Athena|GameState|GetGameStateAthena)))))))

(fn ReportTreeAnalytics
  (bind _attributearray (LocalVariables|GetAttributeArray))
  (if (Networking|IsServer)
    (Utilities|FlowControl|Sequence
      (:then_0
        (bind _returnvalue (Analytics|ConstructAnalyticsAttribute "GameSessionId" (Session|GetGameSessionID (Fortnite|Athena|GameState|GetGameStateAthena))))
        (Utilities|Array|Add _attributearray _returnvalue))
      (:then_1
        (bind _returnvalue_2 (Analytics|ConstructAnalyticsAttribute "Playlist" (Utilities|String|ToString(Name) (Playlist|GetCurrentPlaylistName (Fortnite|Athena|GameState|GetGameStateAthena)))))
        (Utilities|Array|Add _attributearray _returnvalue_2))
      (:then_2
        ;[1 1 1] Build json array string
        (for _array_element (Variables|Default|GetPhysicsLogData)
          (Utilities|Array|Clear (LocalVariables|GetTempStringArray))
          ;;[1 1 1] Wrap elements in quotes
          (for _array_element_2 (Utilities|Array|MakeArray (Utilities|Struct|BreakPhysicsLogAnalyticsStruct _array_element) (.FelledBy_2_076924F44337FDE90F8B5C9C7645CDF2 (Utilities|Struct|BreakPhysicsLogAnalyticsStruct _array_element)) (Utilities|String|ToString(Vector) (.FelledAtLocation_7_87D469DB42BC3085780226AC3B154AD1 (Utilities|Struct|BreakPhysicsLogAnalyticsStruct _array_element))) (Utilities|String|ToString(Float) (.FelledAtMatchTime_17_701A35DB42F8DA247CDC50AC58CCCD4E (Utilities|Struct|BreakPhysicsLogAnalyticsStruct _array_element))) (Utilities|String|ToString(Vector) (.DestroyedAtLocation_33_93747681485FCEB2F9D50B8F2C9BA72B (Utilities|Struct|BreakPhysicsLogAnalyticsStruct _array_element))) (Utilities|String|ToString(Float) (.DestroyedAtMatchTime_31_1DCF4002426A45C86050D4B13C85632F (Utilities|Struct|BreakPhysicsLogAnalyticsStruct _array_element))) (Utilities|String|ToString(Float) (.DamageToPlayers_20_F701392440AAC7AC33A8B2BCAEB3C5D0 (Utilities|Struct|BreakPhysicsLogAnalyticsStruct _array_element))) (Utilities|String|ToString(Float) (.DamageToStructures_22_636FABDE409E1A51BAF808863B05D213 (Utilities|Struct|BreakPhysicsLogAnalyticsStruct _array_element))) (Utilities|String|ToString(Float) (.DamageToVehicles_24_69A07F714B43DF853CA180B441A6A29C (Utilities|Struct|BreakPhysicsLogAnalyticsStruct _array_element))) (Utilities|String|ToString(Float) (.DamageToAI_26_9671D0A943425BB2EC73159C0B19C99D (Utilities|Struct|BreakPhysicsLogAnalyticsStruct _array_element))))
            (Utilities|Array|Add (LocalVariables|GetTempStringArray) (CallFunction|WrapString "\"" _array_element_2 "\"")))
          ;;[1 1 1] Add to instance list
          (Utilities|Array|Add (LocalVariables|GetInstanceRowArray) (CallFunction|WrapString "{" (Utilities|String|JoinStringArray (LocalVariables|GetTempStringArray) " ,") "}")))
        (bind _returnvalue_3 (Analytics|ConstructAnalyticsAttribute "FallenTreeInstances" (CallFunction|WrapString "[" (Utilities|String|JoinStringArray (LocalVariables|GetInstanceRowArray) ",") "]")))
        (Utilities|Array|Add (LocalVariables|GetAttributeArray) _returnvalue_3))
      (:then_3
        ;[1 1 1] Send Analytics
        (Analytics|FireAnalyticsEventBlueprintArray "Core.PhysicsTreeStats" (LocalVariables|GetAttributeArray))))))

(fn ConstructionScript)

(fn WrapString
  (params (Prepend :type String) (String :type String) (Append :type String))
  (return (Utilities|String|Append Prepend String Append)))
