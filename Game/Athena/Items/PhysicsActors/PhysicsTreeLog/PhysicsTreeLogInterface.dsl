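; Blueprint interface for physics tree logs in Athena, declaring a function that returns a tree destruction transform.
; It defines only the GetTreeDestructionTransform signature with a default return, leaving the actual transform to implementing actors.
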(fn GetTreeDestructionTransform
  (return 0))
